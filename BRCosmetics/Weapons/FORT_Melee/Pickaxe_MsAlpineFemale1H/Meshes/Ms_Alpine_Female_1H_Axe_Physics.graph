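# Physics asset for the Ms Alpine Female one-handed axe pickaxe cosmetic, covering the weapon and its dangling ornament.
# One large box body covers the axe on the Melee_TwoHanded_Handle bone, and two smaller boxes cover the dyn_ornament bones.
# Constraints chain ornament 02 to ornament 01 and 01 to the handle, with matching moderate swing limits and tighter twist limits at both joints.
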PhysicsAsset /BRCosmetics/Weapons/FORT_Melee/Pickaxe_MsAlpineFemale1H/Meshes/Ms_Alpine_Female_1H_Axe_Physics
bodies {
 Melee_TwoHanded_Handle [{shapeType=Box, extentX=7.4429, extentY=36.8272, extentZ=64.0429, center=<0 -8.4787 53.7176>, rotation=<rot 0 -90 0>}]
 dyn_ornament_01 [{shapeType=Box, extentX=1.6947, extentY=1.333, extentZ=4, center=<2 0 -0.0581>, rotation=<rot 90 0 0>}]
 dyn_ornament_02 [{shapeType=Box, extentX=5.9668, extentY=1.9482, extentZ=5.8694, center=<2.9 0 -0.0803>, rotation=<rot 90 0 0>}]
}
constraints [
 {bone1Name=dyn_ornament_01, bone2Name=Melee_TwoHanded_Handle, swing1LimitDegrees=45, swing2LimitDegrees=45, twistLimitDegrees=10}
 {bone1Name=dyn_ornament_02, bone2Name=dyn_ornament_01, swing1LimitDegrees=45, swing2LimitDegrees=45, twistLimitDegrees=10}
]
